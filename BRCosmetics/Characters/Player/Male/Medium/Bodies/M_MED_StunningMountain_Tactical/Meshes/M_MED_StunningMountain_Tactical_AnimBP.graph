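# Animation Blueprint for the StunningMountain Tactical male body cosmetic, adding secondary motion and armor correctives to the pose copied from its parent mesh.
# AnimDynamics simulates pouches, grenades and charm, feeding a Control Rig with rocket launcher, melee and minigun flags.
# Bone constraints then place kneepad and shoulder pad bases, and additive offsets correct the pads.
# The update event derives offset alphas from arm rotation, front-end, emote and targeting state, and sets weapon flags.
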AnimBlueprint /BRCosmetics/Characters/Player/Male/Medium/Bodies/M_MED_StunningMountain_Tactical/Meshes/M_MED_StunningMountain_Tactical_AnimBP
target_skeleton=&/M_MED_StunningMountain_Tactical_Skeleton
graphs [
 AnimGraph {nodes [
   n1 = Control Rig!(
    Source=Use cached pose 'AnimDyn'()
    Alpha
    Rocket launcher corr:bool=Get Rocket launcher corr() -> (Rocket launcher corr:bool)
    Melee corr:bool=Get Melee corr() -> (Melee corr:bool)
    Minigun corr:bool=Get Minigun corr() -> (Minigun corr:bool)
   ) -> (Pose) {
    alphaBoolBlend {blendOption=Linear}
    controlRigAssetReference {blueprintRigClass=&/M_MED_StunningMountain_Tactical_CtrlRig.M_MED_StunningMountain_Tactical_CtrlRig_C}
   }
   n2 = "Constraint - dyn_kneepad (PPPP)"!(
    ComponentPose=Local To Component(LocalPose=@n1)
    Alpha=1
    ConstraintWeights_0=0.356
    ConstraintWeights_1:float=0.022
    ConstraintWeights_2:float=0.142
    ConstraintWeights_3:float=0.48
   ) -> (Pose) {
    alphaBoolBlend {blendOption=Linear}
    boneToModify {boneName=dyn_kneepad}
    constraintSetup [
     {targetBone {boneName=calf_l}}
     {targetBone {boneName=deform_knee_in_l}}
     {targetBone {boneName=deform_knee_out_l}}
     {targetBone {boneName=thigh_l}}
    ]
    constraintWeights [1,1,1,1]
   }
   n3 = "Constraint - dyn_shoulderpad_base_r (PP)"!(ComponentPose=@n2, Alpha=1, ConstraintWeights_0=0, ConstraintWeights_1:float=1) -> (Pose) {
    alphaBoolBlend {blendOption=Linear}
    boneToModify {boneName=dyn_shoulderpad_base_r}
    constraintSetup [{targetBone {boneName=clavicle_r}}, {targetBone {boneName=upperarm_twist_01_r}}]
    constraintWeights [1,1]
   }
   n4 = "Constraint - dyn_shoulderpad_base_l (PP)"!(ComponentPose=@n3, Alpha=1, ConstraintWeights_0=0, ConstraintWeights_1:float=1) -> (Pose) {
    alphaBoolBlend {blendOption=Linear}
    boneToModify {boneName=dyn_shoulderpad_base_l}
    constraintSetup [{targetBone {boneName=clavicle_l}}, {targetBone {boneName=upperarm_twist_01_l}}]
    constraintWeights [1,1]
   }
   n5 = "Transform (Modify) Bone - Bone: dyn_shoulderpad_l"!(ComponentPose=@n4, bAlphaBoolEnabled:bool=True, Translation="0.474,0.003,0.250", Rotation="1.776,6.289,-3.573", Scale="1,1,1") -> (Pose) {
    alphaInputType=Bool
    rotationMode=BMM_Additive
    rotationSpace=BCS_BoneSpace
    translationMode=BMM_Additive
    translationSpace=BCS_BoneSpace
    alphaBoolBlend {blendOption=Linear}
    boneToModify {boneName=dyn_shoulderpad_l}
    rotation=<rot 1.776 6.289 -3.573>
    translation=<0.474 0.003 0.25>
   }
   n6 = "Transform (Modify) Bone - Bone: dyn_shoulderpad_r"!(ComponentPose=@n5, bAlphaBoolEnabled:bool=True, Translation="-0.6,0,-0.5", Rotation="-3.498,4.335,0.195", Scale="1,1,1") -> (Pose) {
    alphaInputType=Bool
    rotationMode=BMM_Additive
    rotationSpace=BCS_BoneSpace
    translationMode=BMM_Additive
    translationSpace=BCS_BoneSpace
    alphaBoolBlend {blendOption=Linear}
    boneToModify {boneName=dyn_shoulderpad_r}
    rotation=<rot -3.498 4.335 0.195>
    translation {x=-0.6, z=-0.5}
   }
   n7 = "Transform (Modify) Bone - Bone: dyn_shoulderpad_r"!(
    ComponentPose=@n6
    Alpha:float=Get right shoulder pad arm twist corr() -> (right shoulder pad arm twist corr:double)
    Translation="2.264,0.209,0.133"
    Rotation="20.877,1.577,1.155"
    Scale="1,1,1"
   ) -> (Pose) {
    rotationMode=BMM_Additive
    rotationSpace=BCS_BoneSpace
    translationMode=BMM_Additive
    translationSpace=BCS_BoneSpace
    alphaBoolBlend {blendOption=Linear}
    alphaScaleBiasClamp {bInterpResult=true}
    boneToModify {boneName=dyn_shoulderpad_r}
    rotation=<rot 20.877 1.577 1.155>
    translation=<2.264 0.209 0.133>
   }
   n8 = "Transform (Modify) Bone - Bone: dyn_shoulderpad_l"!(
    ComponentPose=@n7
    Alpha:float=Get left shoulder pad arm twist corr() -> (left shoulder pad arm twist corr:double)
    Translation="-0.589,-0.234,-0.229"
    Rotation="10.842,6.762,-2.982"
    Scale="1,1,1"
   ) -> (Pose) {
    rotationMode=BMM_Additive
    rotationSpace=BCS_BoneSpace
    translationMode=BMM_Additive
    translationSpace=BCS_BoneSpace
    alphaBoolBlend {blendOption=Linear}
    alphaScaleBiasClamp {bInterpResult=true}
    boneToModify {boneName=dyn_shoulderpad_l}
    rotation=<rot 10.842 6.762 -2.982>
    translation=<-0.589 -0.234 -0.229>
   }
   n9 = "Transform (Modify) Bone - Bone: dyn_shoulderpad_r"(
    ComponentPose=@n8
    Alpha=Get right shoulder pad up corr() -> (right shoulder pad up corr:double)
    Translation="-0.755,-8.787,-2.3"
    Rotation="-8.351,-6.654,11.424"
    Scale="1,1,1"
   ) {
    rotationMode=BMM_Additive
    rotationSpace=BCS_BoneSpace
    translationMode=BMM_Additive
    translationSpace=BCS_BoneSpace
    alphaBoolBlend {blendOption=Linear}
    boneToModify {boneName=dyn_shoulderpad_r}
    rotation=<rot -8.351 -6.654 11.424>
    translation=<-0.755 -8.787 -2.3>
   }
   n10 = "Transform (Modify) Bone - Bone: dyn_shoulderpad_l"(
    ComponentPose=@n9
    Alpha=Get left shoulder pad up corr() -> (left shoulder pad up corr:double)
    Translation="-1.182,6.897,2.3"
    Rotation="7.770,14.099,20.123"
    Scale="1,1,1"
   ) {
    rotationMode=BMM_Additive
    rotationSpace=BCS_BoneSpace
    translationMode=BMM_Additive
    translationSpace=BCS_BoneSpace
    alphaBoolBlend {blendOption=Linear}
    boneToModify {boneName=dyn_shoulderpad_l}
    rotation=<rot 7.7699 14.0989 20.1229>
    translation=<-1.182 6.897 2.3>
   }
   n11 = Get Frontend corr() -> (Frontend corr:bool)
   n12 = "Transform (Modify) Bone - Bone: dyn_shoulderpad_r"!(ComponentPose=@n10, bAlphaBoolEnabled:bool=@n11, Translation="-1.163,0.134,-0.161", Rotation="-14.058,-2.562,-2.604", Scale="1,1,1") -> (Pose) {
    alphaInputType=Bool
    rotationMode=BMM_Additive
    rotationSpace=BCS_BoneSpace
    translationMode=BMM_Additive
    translationSpace=BCS_BoneSpace
    alphaBoolBlend {blendInTime=1, blendOption=Linear, blendOutTime=0.3}
    boneToModify {boneName=dyn_shoulderpad_r}
    rotation=<rot -14.058 -2.5619 -2.6041>
    translation=<-1.163 0.134 -0.161>
   }
   n13 = "Transform (Modify) Bone - Bone: dyn_shoulderpad_l"!(ComponentPose=@n12, bAlphaBoolEnabled:bool=@n11, Translation="0.135,0.083,-0.217", Rotation="-8.782,-8.121,1.216", Scale="1,1,1") -> (Pose) {
    alphaInputType=Bool
    rotationMode=BMM_Additive
    rotationSpace=BCS_BoneSpace
    translationMode=BMM_Additive
    translationSpace=BCS_BoneSpace
    alphaBoolBlend {blendInTime=0.8, blendOption=Linear, blendOutTime=0.3}
    boneToModify {boneName=dyn_shoulderpad_l}
    rotation=<rot -8.782 -8.121 1.216>
    translation=<0.135 0.083 -0.217>
   }
   Output Pose(Result=Component To Local(ComponentPose=@n13)) {layerGroup=None}
   n14 = Anim Dynamics - Bone: dyn_pouch_r(ComponentPose=Local To Component(LocalPose=Copy Pose From Mesh() {bUseAttachedParent=true}), Alpha=1, ExternalForce="0,0,0") {
    angularDampingOverride=0.9999
    angularSpringConstant=4000
    bAngularSpring=true
    bOverrideAngularDamping=true
    simulationSpace=RootRelative
    alphaBoolBlend {blendOption=Linear}
    boundBone {boneName=dyn_pouch_r}
    componentAppliedLinearAccClamp=<3000 3000 3000>
    componentLinearAccScale=<1 1 1>
    componentLinearVelScale=<1 1 1>
    physicsBodyDefinitions [{
      boundBone {boneName=dyn_pouch_r}
      boxExtents=<15 15 20>
      constraintSetup {twistAxis=AxisZ, angularLimitsMax=<4 5 0>, angularLimitsMin=<-14 -5 0>, angularTarget=<0 0 0>}
      localJointOffset=<0 0 -1>
     }]
   }
   n15 = Anim Dynamics - Bone: dyn_pouch01_bk_l(ComponentPose=@n14, Alpha=1, ExternalForce="0,0,0") {
    angularDampingOverride=0.9999
    angularSpringConstant=4000
    bAngularSpring=true
    bOverrideAngularDamping=true
    simulationSpace=RootRelative
    alphaBoolBlend {blendOption=Linear}
    boundBone {boneName=dyn_pouch01_bk_l}
    componentAppliedLinearAccClamp=<3000 3000 3000>
    componentLinearAccScale=<1 1 1>
    componentLinearVelScale=<1 1 1>
    physicsBodyDefinitions [{
      boundBone {boneName=dyn_pouch01_bk_l}
      boxExtents=<18 20 20>
      constraintSetup {twistAxis=AxisZ, angularLimitsMax=<4 5 0>, angularLimitsMin=<-12 -5 0>, angularTarget=<0 0 0>}
      localJointOffset=<0 0 -1>
     }]
   }
   n16 = Anim Dynamics - Bone: dyn_pouch_l(ComponentPose=@n15, Alpha=1, ExternalForce="0,0,0") {
    angularDampingOverride=0.9999
    angularSpringConstant=4000
    bAngularSpring=true
    bOverrideAngularDamping=true
    simulationSpace=RootRelative
    alphaBoolBlend {blendOption=Linear}
    boundBone {boneName=dyn_pouch_l}
    componentAppliedLinearAccClamp=<3000 3000 3000>
    componentLinearAccScale=<1 1 1>
    componentLinearVelScale=<1 1 1>
    physicsBodyDefinitions [{
      boundBone {boneName=dyn_pouch_l}
      boxExtents=<18 15 20>
      constraintSetup {twistAxis=AxisZ, angularLimitsMax=<4 3 0>, angularLimitsMin=<-14 -3 0>, angularTarget=<0 0 0>}
      localJointOffset=<0 0 -1>
     }]
   }
   n17 = Anim Dynamics - Bone: dyn_pouch02_bk_l(ComponentPose=@n16, Alpha=1, ExternalForce="0,0,0") {
    angularDampingOverride=0.9999
    angularSpringConstant=4000
    bAngularSpring=true
    bOverrideAngularDamping=true
    simulationSpace=RootRelative
    alphaBoolBlend {blendOption=Linear}
    boundBone {boneName=dyn_pouch02_bk_l}
    componentAppliedLinearAccClamp=<3000 3000 3000>
    componentLinearAccScale=<1 1 1>
    componentLinearVelScale=<1 1 1>
    physicsBodyDefinitions [{
      boundBone {boneName=dyn_pouch02_bk_l}
      boxExtents=<15 15 20>
      constraintSetup {twistAxis=AxisZ, angularLimitsMax=<4 3 0>, angularLimitsMin=<-14 -3 0>, angularTarget=<0 0 0>}
      localJointOffset=<0 0 -1>
     }]
   }
   n18 = Anim Dynamics - Bone: dyn_grenade01_r(ComponentPose=@n17, Alpha=1, ExternalForce="0,0,0") {
    angularDampingOverride=0.9994
    angularSpringConstant=4000
    bAngularSpring=true
    bOverrideAngularDamping=true
    simulationSpace=RootRelative
    alphaBoolBlend {blendOption=Linear}
    boundBone {boneName=dyn_grenade01_r}
    componentAppliedLinearAccClamp=<3000 3000 3000>
    componentLinearAccScale=<0.3 0.3 1>
    componentLinearVelScale=<0.7 0.7 1>
    physicsBodyDefinitions [{
      boundBone {boneName=dyn_grenade01_r}
      boxExtents=<23 23 23>
      constraintSetup {twistAxis=AxisZ, angularLimitsMax=<6 10 0>, angularLimitsMin=<-15 -10 0>, angularTarget=<0 0 0>}
      localJointOffset=<0 0 -3>
     }]
   }
   n19 = Anim Dynamics - Bone: dyn_grenade02_r(ComponentPose=@n18, Alpha=1, ExternalForce="0,0,0") {
    angularDampingOverride=0.9994
    angularSpringConstant=4000
    bAngularSpring=true
    bOverrideAngularDamping=true
    simulationSpace=RootRelative
    alphaBoolBlend {blendOption=Linear}
    boundBone {boneName=dyn_grenade02_r}
    componentAppliedLinearAccClamp=<3000 3000 3000>
    componentLinearAccScale=<0.3 0.3 1>
    componentLinearVelScale=<0.7 0.7 1>
    physicsBodyDefinitions [{
      boundBone {boneName=dyn_grenade02_r}
      boxExtents=<25 25 25>
      constraintSetup {twistAxis=AxisZ, angularLimitsMax=<6 10 0>, angularLimitsMin=<-15 -10 0>, angularTarget=<0 0 0>}
      localJointOffset=<0 0 -2.3>
     }]
   }
   n20 = Anim Dynamics - Bone: dyn_belt_pouch_l(ComponentPose=@n19, Alpha=1, ExternalForce="0,0,0") {
    angularDampingOverride=0.9994
    angularSpringConstant=4000
    bAngularSpring=true
    bOverrideAngularDamping=true
    bUsePlanarLimit=false
    bUseSphericalLimits=true
    simulationSpace=RootRelative
    alphaBoolBlend {blendOption=Linear}
    boundBone {boneName=dyn_belt_pouch_l}
    componentAppliedLinearAccClamp=<3000 3000 3000>
    componentLinearAccScale=<0.3 0.3 1>
    componentLinearVelScale=<0.7 0.7 1>
    physicsBodyDefinitions [{
      sphereCollisionRadius=4
      boundBone {boneName=dyn_belt_pouch_l}
      boxExtents=<30 30 30>
      constraintSetup {twistAxis=AxisZ, angularLimitsMax=<15 1 0>, angularLimitsMin=<6 -2 0>, angularTarget=<0 0 0>}
      localJointOffset=<-0.8865 6.7617 -10.1786>
     }]
   }
   n21 = Anim Dynamics - Bone: dyn_charm_belt(ComponentPose=@n20, Alpha=1, ExternalForce="0,0,0") {
    angularDampingOverride=0.999
    angularSpringConstant=3000
    bAngularSpring=true
    bOverrideAngularDamping=true
    bUsePlanarLimit=false
    simulationSpace=RootRelative
    alphaBoolBlend {blendOption=Linear}
    boundBone {boneName=dyn_charm_belt}
    componentAppliedLinearAccClamp=<3000 3000 3000>
    componentLinearAccScale=<1 1 1>
    componentLinearVelScale=<1 1 1>
    physicsBodyDefinitions [{
      sphereCollisionRadius=2
      boundBone {boneName=dyn_charm_belt}
      boxExtents=<15 15 15>
      constraintSetup {angularTargetAxis=AxisZ, twistAxis=AxisZ, angularLimitsMax=<2 10 0>, angularLimitsMin=<-20 -10 0>, angularTarget=<0 0 0>}
      localJointOffset=<0 0 -1.2>
     }]
   }
   n22 = Anim Dynamics - Bone: dyn_charm(ComponentPose=@n21, Alpha=1, ExternalForce="0,0,0") {
    angularDampingOverride=0.999
    angularSpringConstant=3000
    bAngularSpring=true
    bOverrideAngularDamping=true
    bUseSphericalLimits=true
    simulationSpace=RootRelative
    alphaBoolBlend {blendOption=Linear}
    boundBone {boneName=dyn_charm}
    componentAppliedLinearAccClamp=<3000 3000 3000>
    componentLinearAccScale=<0.7 0.7 0.7>
    componentLinearVelScale=<1 1 1>
    physicsBodyDefinitions [{
      collisionType=CustomSphere
      sphereCollisionRadius=2.5
      boundBone {boneName=dyn_charm}
      boxExtents=<23 23 23>
      constraintSetup {angularTargetAxis=AxisZ, twistAxis=AxisZ, angularLimitsMax=<5 15 50>, angularLimitsMin=<-15 -15 -50>, angularTarget=<0 0 0>}
      localJointOffset=<-0.4115 0.5669 -2.8223>
     }]
    planarLimits [{drivingBone {boneName=pelvis}, planeTransform {location=<-0.142 7.8138 4.1225>, rotation=<rot -5.4161 -13.4271 68.7212>}}]
   }
   Save cached pose 'AnimDyn'(Pose=Component To Local(ComponentPose=@n22)) {cacheName=AnimDyn}
  ]}
 EventGraph {nodes [
   Event BlueprintUpdateAnimation()
   > n1 = Sequence!(execute) -> (then_0, then_1, then_2:Exec)
   n2 = Cast To FortPlayerAnimInstance(execute=@n1.then_1, Object=GetAnimInstance(self=GetOwnerSkeletalMeshComponent()))
   > Set Rocket launcher corr(Rocket launcher corr:bool=false) -> (Output_Get:bool)
   > Set Melee corr(Melee corr:bool=false) -> (Output_Get:bool)
   > Set Minigun corr(Minigun corr:bool=false) -> (Output_Get:bool)
   > n3 = Switch on EFortWeaponCoreAnimation!(execute, Selection:EFortWeaponCoreAnimation Enum=Get WeaponCoreAnim(self=@n2)) -> (
    Melee:Exec
    Pistol:Exec
    Shotgun:Exec
    PaperBlueprint:Exec
    Rifle:Exec
    MeleeOneHand:Exec
    MachinePistol:Exec
    RocketLauncher:Exec
    GrenadeLauncher:Exec
    GoingCommando:Exec
    AssaultRifle:Exec
    TacticalShotgun:Exec
    SniperRifle:Exec
    TrapPlacement:Exec
    ShoulderLauncher:Exec
    AbilityDecoTool:Exec
    Crossbow:Exec
    C4:Exec
    RemoteControl:Exec
    DualWield:Exec
    AR_BullPup:Exec
    AR_ForwardGrip:Exec
    MedPackPaddles:Exec
    SMG_P90:Exec
    AR_DrumGun:Exec
    Consumable_Small:Exec
    Consumable_Large:Exec
    Balloon:Exec
    MountedTurret:Exec
    CreativeTool:Exec
    ExplosiveBow:Exec
    AshtonIndigo:Exec
    AshtonChicago:Exec
    MeleeDualWield:Exec
    Unarmed:Exec
   )
   Set Minigun corr(execute=@n3.GoingCommando, Minigun corr=true)
   Set Melee corr(execute=[@n3.Melee, @n3.MeleeDualWield, @n3.MeleeOneHand], Melee corr=true)
   Set Rocket launcher corr(execute=@n3.RocketLauncher, Rocket launcher corr=true)
   n4 = Event OnCharacterCustomizationLoaded()
   > n5 = Set Skeletal Mesh(Skeletal Mesh:Skeletal Mesh Component Object Reference=@n4.CustomBaseMesh) -> (Output_Get:Skeletal Mesh Component Object Reference)
   Set left shoulder pad arm twist corr(
    execute=@n1.then_0
    left shoulder pad arm twist corr:double=GetAlphaFromReferenceTransformDelta(SkeletalMeshComponent=@n5, BoneName=upperarm_l, BaseName=spine_05, DeltaComponentType=EulerX, InputRange=(X=20,Y=3), TwistAxis="1,0,0")
   ) -> (Output_Get:double)
   > Set right shoulder pad arm twist corr(right shoulder pad arm twist corr:double=SelectFloat(
     A=0
     B=GetAlphaFromReferenceTransformDelta(SkeletalMeshComponent=@n5, BoneName=upperarm_r, BaseName=spine_05, DeltaComponentType=EulerX, InputRange=(X=40,Y=3), TwistAxis="1,0,0")
     bPickA=Get bIsTargeting()
    )) -> (Output_Get:double)
   > Set right shoulder pad up corr(right shoulder pad up corr:double=GetAlphaFromReferenceTransformDelta(SkeletalMeshComponent=@n5, BoneName=upperarm_r, BaseName=clavicle_r, DeltaComponentType=EulerY, InputRange=(X=25,Y=60), TwistAxis="1,0,0")) -> (Output_Get:double)
   > Set left shoulder pad up corr(left shoulder pad up corr:double=GetAlphaFromReferenceTransformDelta(SkeletalMeshComponent=@n5, BoneName=upperarm_l, BaseName=clavicle_l, DeltaComponentType=EulerY, InputRange=(X=25,Y=60), TwistAxis="1,0,0")) -> (Output_Get:double)
   n6 = Get bIsPlayingEmote()
   > Set Frontend corr(Frontend corr:bool=AND Boolean(A=Get bIsInFrontEnd(), B=NOT Boolean(A=@n6))) -> (Output_Get:bool)
  ]}
]
variables [
 Frontend corr {type=boolean}
 left shoulder pad arm twist corr {type=number}
 left shoulder pad up corr {type=number}
 Melee corr {type=boolean}
 Minigun corr {type=boolean}
 right shoulder pad arm twist corr {type=number}
 right shoulder pad up corr {type=number}
 Rocket launcher corr {type=boolean}
 Skeletal Mesh {type=SkeletalMeshComponent}
]
